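# Prototype looping music SoundCue for the RedLight gameplay in TimberStakeMap, routed to the base music sound class.
# A single looping wave player plays MX_Proto_RedLight_Loop_01 at a lowered pitch.
# Overridden attenuation uses a non-spatialized box volume with natural-sound falloff, occlusion and light reverb wet levels.
# The cue's volume multiplier is zero, so it is silent as authored.
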SoundCue /TimberStakeMapContent/Gameplay/RedLight/Assets/Audio/Proto/MX_Proto_RedLIght_Loop_Cue
root=WavePlayer_0
nodes [{
  id=WavePlayer_0
  properties {bLooping=true, soundWaveAssetPtr=&/MX_Proto_RedLight_Loop_01}
 }]
properties {
 bOverrideAttenuation=true
 maxDistance=2700
 pitchMultiplier=0.6
 soundClassObject=/Game/Sounds/FortSoundClassesAndModes/Fort_11_Base_Music.Fort_11_Base_music
 volumeMultiplier=0
 attenuationOverrides {
  attenuationShape=Box
  bEnableOcclusion=true
  bSpatialize=false
  distanceAlgorithm=NaturalSound
  falloffDistance=1200
  reverbWetLevelMax=0.2
  reverbWetLevelMin=0.1
  attenuationShapeExtents=<256 1500 184>
 }
}
